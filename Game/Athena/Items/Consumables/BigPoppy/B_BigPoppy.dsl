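; Big Poppy consumable actor: a balloon that inflates on spawn with a sound, then pops.
; A timeline scales the balloon and calls Pop when finished; touching a FortStaticMeshActor also pops it.
; Balloon collision destroys building pieces and props, sparing other Big Poppies, and launches players along their vector to its center, scaled by PlayerPushFactor.
; Pop plays particles and sound behind a cosmetic-events check and destroys the actor.
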(event EventBeginPlay
  (Utilities|FlowControl|Sequence
    (:then_0
      (bind (_direction _scale) (Timeline|Timeline_0 PlayFromStart)
        (:Update
          (Transformation|SetRelativeScale3D (Variables|Default|GetBalloon) (Math|Conversions|ToVector(Float) _scale)))
        (:Finished
          (CallFunction|Pop))))
    (:then_1
      (Audio|PlaySoundatLocation "/Game/Athena/Sounds/Items/BigPoopy/Athena_BigPoppy_Inflate_Cue.Athena_BigPoppy_Inflate_Cue" (Transformation|GetWorldLocation (Variables|Default|GetBalloonCollision))))))

(event OnComponentBeginOverlap(BalloonCollision)
  (params (OverlappedComponent :type Primitive%20Component%20Object%20Reference) (OtherActor :type Actor%20Object%20Reference) (OtherComp :type Primitive%20Component%20Object%20Reference) (OtherBodyIndex :type Integer) (bFromSweep :type Boolean) (SweepResult :type Hit%20Result%20Structure%20%28by%20ref%29))
  (Utilities|Casting|CastToBuildingActor OtherActor
    (:then
      (Utilities|Casting|CastToB_BigPoppy OtherActor
        (:CastFailed
          (Utilities|Casting|CastToBuildingWall OtherActor
            (:then
              _)
            (:CastFailed
              (Utilities|Casting|CastToBuildingFloor OtherActor
                (:then
                  _)
                (:CastFailed
                  (Utilities|Casting|CastToBuildingRoof OtherActor
                    (:then
                      _)
                    (:CastFailed
                      (Utilities|Casting|CastToBuildingStairs OtherActor
                        (:then
                          _)
                        (:CastFailed
                          (Utilities|Casting|CastToBuildingCorner OtherActor
                            (:then
                              _)
                            (:CastFailed
                              (Utilities|Casting|CastToBuildingPillar OtherActor
                                (:then
                                  _)
                                (:CastFailed
                                  (Utilities|Casting|CastToBuildingDeco OtherActor
                                    (:then
                                      _)
                                    (:CastFailed
                                      (Utilities|Casting|CastToBuildingProp OtherActor
                                        (:then
                                          _)
                                        (:CastFailed
                                          (Utilities|Casting|CastToBuildingContainer OtherActor)))))))))))))))))
          (Actor|DestroyActor OtherActor)
          (Audio|PlaySoundatLocation "/Game/Sounds/Fort_Traps/JumpPad/JumpPad_Destroyed_Cue.JumpPad_Destroyed_Cue" (Transformation|GetActorLocation OtherActor) "0, 0, 0" 1.0 1.0 0.0 "/Game/Sounds/Attenuation/Fort_Survivor_Att.Fort_Survivor_Att"))))
    (:CastFailed
      (Utilities|Casting|CastToFortStaticMeshActor OtherActor
        (:then
          (CallFunction|Pop))
        (:CastFailed
          (bind _asfort_player_pawn_athena (Utilities|Casting|CastToFortPlayerPawnAthena OtherActor))
          (Character|LaunchCharacter _asfort_player_pawn_athena (* (- (Transformation|GetWorldLocation (Variables|Default|GetBalloonCollision)) (Transformation|GetActorLocation _asfort_player_pawn_athena)) (Variables|Default|GetPlayerPushFactor)) false false)
          (Audio|PlaySoundatLocation "/Game/Sounds/Fort_Test_Sounds/World_Objects/Bouncing/TwangRubber1_Cue.TwangRubber1_Cue" (Transformation|GetActorLocation _asfort_player_pawn_athena)))))))

(event Collision|EventActorBeginOverlap
  (params (OtherActor :type Actor%20Object%20Reference)))

(event EventTick
  (params (DeltaSeconds :type Float%20%28single-precision%29)))

(fn Pop
  (bind _ballooncollision (Variables|Default|GetBalloonCollision))
  (bind _returnvalue (Transformation|GetWorldLocation _ballooncollision))
  (Utilities|FlowControl|Sequence
    (:then_0
      (Networking|CanExecuteCosmeticEvents
        (:True
          (bind _returnvalue_2 (Effects|SpawnEmitteratLocation "/Game/Athena/Items/Consumables/BigPoppy/P_Athena_BigPoppy_Pop.P_Athena_BigPoppy_Pop" (Math|Vector|MakeVector (.x _returnvalue) (.y _returnvalue) (- (.z _returnvalue) 256))))
          (Transformation|SetWorldScale3D _returnvalue_2 "2.500000,2.500000,2.500000")
          (Audio|PlaySoundatLocation "/Game/Sounds/Fort_UI_Sounds/HomeBase/HomeBase_Generic_Building_Pop_Cue.HomeBase_Generic_Building_Pop_Cue" _returnvalue)
          (Components|DestroyComponent (Variables|Default|GetCapsuleHalf)))))
    (:then_1
      (Actor|DestroyActor))))

(fn ConstructionScript)
