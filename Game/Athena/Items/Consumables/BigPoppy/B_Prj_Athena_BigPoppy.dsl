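; Thrown Big Poppy grenade projectile that spawns a B_BigPoppy actor and explodes.
; A fixed fuse after BeginPlay, damage or a bounce on water triggers Explode; reaching the bounce limit only stores the hit.
; Capsule overlaps fake bounces by reversing horizontal velocity, forcing it upward and adjusting gravity scale.
; StoredHit's OnRep sets the location for explosion effects; the fuse also shows a radius indicator before detonating, tinted for non-teammates.
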(event OnComponentBeginOverlap(OverlapCapsule)
  (params (OverlappedComponent :type Primitive%20Component%20Object%20Reference) (OtherActor :type Actor%20Object%20Reference) (OtherComp :type Primitive%20Component%20Object%20Reference) (OtherBodyIndex :type Integer) (bFromSweep :type Boolean) (SweepResult :type Hit%20Result%20Structure%20%28by%20ref%29))
  (bind (_bblockinghit _binitialoverlap _time _distance _location _impactpoint _normal _impactnormal _physmat _hitactor) (Collision|BreakHitResult SweepResult))
  (bind _projectilemovementcomponent (Variables|Movement|GetProjectileMovementComponent))
  ;[1 1 1] If we overlap a pawn, fake a bounce.
  (bind _aspawn (Utilities|Casting|CastToPawn _hitactor))
  (if (Utilities|NotEqual(Object) _aspawn (Game|GetInstigator))
    (Variables|/Script/Engine.ProjectileMovementComponent|SetProjectileGravityScale (Variables|Default|GetGravityScaleAgainstPawn) _projectilemovementcomponent) ; Hit a pawn, scale gravity
    (else
      (Variables|/Script/Engine.ProjectileMovementComponent|SetProjectileGravityScale (Variables|Default|GetGravityScaleAgainstElse) _projectilemovementcomponent) ; Hit else, scale gravity
    ))
  (Variables|/Script/Engine.MovementComponent|SetVelocity (* (Math|Vector|MakeVector (.x (* (Variables|/Script/Engine.MovementComponent|GetVelocity (Variables|Movement|GetProjectileMovementComponent)) -1)) (.y (* (Variables|/Script/Engine.MovementComponent|GetVelocity (Variables|Movement|GetProjectileMovementComponent)) -1)) (Math|Float|Absolute(Float) (.z (* (Variables|/Script/Engine.MovementComponent|GetVelocity (Variables|Movement|GetProjectileMovementComponent)) -1)))) (Variables|/Script/Engine.ProjectileMovementComponent|GetBounciness (Variables|Movement|GetProjectileMovementComponent))) (Variables|Movement|GetProjectileMovementComponent))
  (Variables|Default|SetCurrentNumberOfBounces (+ (Variables|Default|GetCurrentNumberOfBounces) 1))
  (if (>= (Variables|Default|GetCurrentNumberOfBounces) (Variables|Default|GetNumberOfBouncesTillExplode))
    (Utilities|FlowControl|DoOnce false
      (:Completed
        (Variables|Default|SetStoredHit SweepResult)))))

(event EventOnBounce
  (params (Hit :type Hit%20Result%20Structure%20%28by%20ref%29))
  (bind (_bblockinghit _binitialoverlap _time _distance _location _impactpoint _normal _impactnormal _physmat _hitactor) (Collision|BreakHitResult Hit))
  ;[1 1 1] OnBounce - After X bounces set to explode
  (bind _aspawn (Utilities|Casting|CastToPawn _hitactor)
    (:then
      (if (Utilities|NotEqual(Object) _aspawn (Game|GetInstigator))
        (Variables|/Script/Engine.ProjectileMovementComponent|SetProjectileGravityScale (Variables|Default|GetGravityScaleAgainstPawn) (Variables|Movement|GetProjectileMovementComponent)) ; Hit a pawn, scale gravity
      ))
    (:CastFailed
      (if (Utilities|Equal(Object) _physmat "/Game/FortressPhysicalMaterials/Water.Water")
        (Variables|Default|SetStoredHit Hit)
        (CallFunction|Explode)
        (else
          _))))
  (Variables|Default|SetCurrentNumberOfBounces (+ (Variables|Default|GetCurrentNumberOfBounces) 1))
  (Audio|PlaySoundatLocation (Variables|Sounds|GetCue_Bounce) _location)
  (if (>= (Variables|Default|GetCurrentNumberOfBounces) (Variables|Default|GetNumberOfBouncesTillExplode))
    (Utilities|FlowControl|DoOnce false
      (:Completed
        (Variables|Default|SetStoredHit Hit)))))

(event Custom|Stop_Rotation
  ;[1 1 1] Stop Rotating Movement
  (Components|Movement|SetUpdatedComponent (Variables|Default|GetRotatingMovement) 0))

(event EventOnStop
  (params (Hit :type Hit%20Result%20Structure%20%28by%20ref%29))
  ;[0.7116 0.838 1] OnStop - Stop the rotation
  (CallFunction|StopRotation))

(event EventBeginPlay
  (bind _effect_distance (Variables|B_Prj_Grenade|GetEffect_Distance))
  ;[1 1 1] BeginPlay - Start Fuse
  (Utilities|FlowControl|Sequence
    (:then_0
      (Networking|CanExecuteCosmeticEvents
        (:True
          (Audio|Components|Audio|SetSound (Variables|B_Prj_Grenade|GetGrenadeFuse_AudioComponent) (Variables|Default|GetCue_GrenadeFuseSound)))))
    (:then_1
      (Utilities|FlowControl|Delay 2.75 :!pinned true) ; Time Before Explosion
      (Variables|Default|SetStoredHit (Collision|MakeHitResult false false 0.0 0.0 (Transformation|GetActorLocation) (Transformation|GetActorLocation) (Transformation|GetActorLocation) (Transformation|GetActorLocation) 0 (Actor|GetParentActor)))
      (Transformation|SetWorldRotation _effect_distance "90.000000,0.000000, 0" false false)
      (Components|Activation|Activate _effect_distance false)
      (Effects|Components|ParticleSystem|SetVectorParameter _effect_distance "Size" (Math|Conversions|ToVector(Float) (Variables|Default|GetExplosionRadius)))
      (if (not (Team|OnSameTeam (Game|GetInstigatorController) (Game|GetPlayerController 0)))
        (Effects|Components|ParticleSystem|SetVectorParameter _effect_distance "Color" "1.500000,0.110000,0.190000"))
      (Utilities|FlowControl|DoOnce false
        (:Completed
          (CallFunction|Explode))))))

(event Game|Damage|EventAnyDamage
  (params (Damage :type Float%20%28single-precision%29) (DamageType :type Damage%20Type%20Object%20Reference) (InstigatedBy :type Controller%20Object%20Reference) (DamageCauser :type Actor%20Object%20Reference))
  ;[1 1 1] BeginPlay - Start Fuse
  (Utilities|FlowControl|Delay 0.1)
  (Utilities|FlowControl|DoOnce false
    (:Completed
      (CallFunction|Explode))))

(event EventOnExploded
  (params (HitActors :type Array%20of%20Actor%20Object%20References) (HitResults :type Array%20of%20Hit%20Result%20Structures))
  ;[1 1 1] OnExploded - Emitter, Sound, Camera Shake
  (Networking|CanExecuteCosmeticEvents
    (:True
      (Effects|SpawnEmitteratLocation (Variables|Default|GetP_Explosion) (Variables|Default|GetHitLocation) (Math|Rotator|MakeRotator 0.0 90))
      (Utilities|FlowControl|Sequence
        (:then_0
          (Camera|PlayWorldCameraShake "/Game/Abilities/Player/Commando/Actives/FragGrenade/B_GrenadeCameraShake.B_GrenadeCameraShake_C" (Variables|Default|GetHitLocation) 200 800))
        (:then_1
          (if (Pawn|IsLocallyControlled (Game|GetInstigator))
            (Audio|PlaySoundatLocation (Variables|Sounds|GetCue_CloseSound) (Variables|Default|GetHitLocation)) ; Play Close Explosion
            (Audio|Components|Audio|FadeOut (Variables|B_Prj_Grenade|GetGrenadeFuse_AudioComponent) 0.2 0.0 "Linear" :!class /Script/Engine.AudioComponent) ; Fade Out Fuse Sound
            (else
              (Audio|PlaySoundatLocation (Variables|Sounds|GetCue_DistantSound) (Variables|Default|GetHitLocation)) ; Play Distant Explosion
            )))
        (:then_2
          (Fortnite|PlayLocalForceFeedbackatLocationMulti (Variables|Default|GetExplosionForceFeedbackNear) (Variables|Default|GetExplosionForceFeedbackFar) (Variables|Default|GetHitLocation) (Variables|Default|GetExplosionRadius) (* (Variables|Default|GetExplosionRadius) 3) "Explosion")
          (Development|SetHiddeninGame (Variables|B_Prj_Grenade|GetMesh) true false))))))

(fn Explode
  (Utilities|FlowControl|DoOnce false
    (:Completed
      (Game|SpawnActorfromClass "/Game/Athena/Items/Consumables/BigPoppy/B_BigPoppy.B_BigPoppy_C" (Math|Conversions|ToTransform(Vector) (Math|Vector|MakeVector (.x (Transformation|GetActorLocation self)) (.y (Transformation|GetActorLocation self)) (- (.z (Transformation|GetActorLocation self)) 12))) "AdjustIfPossibleButAlwaysSpawn" "OverrideRootScale" 0 0 0)
      (Delivery|DoExplosion))))

(fn OnRep_StoredHit
  (bind (_bblockinghit _binitialoverlap _time _distance _location) (Collision|BreakHitResult (Variables|Default|GetStoredHit)))
  (Variables|Default|SetHitLocation _location))

(fn ConstructionScript
  (Networking|CanExecuteCosmeticEvents
    (:True
      (Components|Movement|SetUpdatedComponent (Variables|Default|GetRotatingMovement) (Variables|B_Prj_Grenade|GetMesh))
      (Components|Capsule|SetCapsuleSize (Variables|Default|GetOverlapCapsule) (Variables|/Script/Engine.CapsuleComponent|GetCapsuleRadius (Variables|Collision|GetCapsuleComponent)) (Variables|/Script/Engine.CapsuleComponent|GetCapsuleHalfHeight (Variables|Collision|GetCapsuleComponent)) true))))
