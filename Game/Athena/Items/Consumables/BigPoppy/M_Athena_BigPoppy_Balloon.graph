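# Surface material for the balloon of the BigPoppy consumable in Athena, using Default Lit shading.
# It is enabled for skeletal meshes and mesh deformers.
# The graph holds a single sample of the balloon texture, tiled four times along U only, with no connection to a material output shown.
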Material /Game/Athena/Items/Consumables/BigPoppy/M_Athena_BigPoppy_Balloon
graph {
 properties {bUsedWithMeshDeformer=true, bUsedWithSkeletalMesh=true, pixelDepthOffsetMode=PDOM_Legacy, usedShadingModels=Default Lit}
 nodes [TextureSample(UVs=TextureCoordinate() {uTiling=4}) {texture=&/T_Athena_BigPoppy_Balloon}]
}
